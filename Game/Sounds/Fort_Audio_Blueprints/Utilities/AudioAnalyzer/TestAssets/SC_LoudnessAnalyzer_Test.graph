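# Test SoundCue for the Fort audio loudness analyzer utility.
# A single looping wave player plays what is named as an ITU 1770-2 compliance reference tone: stereo 1 kHz at -23 LKFS.
# Its output is sent at full level to the AB_LoudnessAnalyzer_Main audio bus, giving the analyzer a known reference signal.
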SoundCue /Game/Sounds/Fort_Audio_Blueprints/Utilities/AudioAnalyzer/TestAssets/SC_LoudnessAnalyzer_Test
root=WavePlayer_0
nodes [{
  id=WavePlayer_0
  properties {bLooping=true, soundWaveAssetPtr=&/1770-2_Comp_23LKFS_1000Hz_2ch}
 }]
properties {
 maxDistance=2097152
 volumeMultiplier=1
 busSends [{
   audioBus=/Game/Sounds/Fort_Audio_Blueprints/Utilities/AudioAnalyzer/TestAssets/AudioBusses/AB_LoudnessAnalyzer_Main
   minSendLevel=0
   sendLevel=1
   soundSourceBus=null
  }]
}
